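# Niagara system for the Season 30 frontend ambience: warm car lens flares, two light shafts, drifting foreground fog and purple-pink "mouth" smoke.
# Car_LensFlare runs on GPU, placing six flares by partition and driving visibility through an occlusion scratch module and size and brightness through camera angle.
# Car_LensFlare_LOW is a CPU variant without occlusion or camera scaling.
# Waveforms pulse flare and shaft alpha; fog and smoke use SubUV-animated sprites.
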NiagaraSystem /Game/NeverCook/FX/FrontEndFX/S30/NS_S30_Frontend_Ambient
dependencies {
 usedDataInterfaces [
  &NiagaraDataInterfaceCurve
  &NiagaraDataInterfaceSpriteRendererInfo
  &NiagaraDataInterfaceArrayFloat3
  &NiagaraDataInterfaceArrayFloat2
  /Script/Niagara.NiagaraDataInterfaceOcclusion
  &NiagaraDataInterfaceCamera
  &NiagaraDataInterfaceArrayColor
  &NiagaraDataInterfaceArrayFloat
 ]
 usedDynamicInputs [&Vector2DFromFloat]
 usedModules [&SystemState, &Timeline]
}
emitters {
 Car_LensFlare {
  simTarget=GPUComputeSim
  emitterUpdateScript [
   &EmitterState()
   &SpawnBurst_Instantaneous(Spawn Count=6, Use Spawn Probability=false, UseLoopCountLimit=true)
  ]
  particleSpawnScript [
   &PartitionParticles(Number Of Partitions=6)
   &InitializeParticle(
    Color=<rgba 1 0.5764 0.2812 0.25>
    Mesh Scale Mode=::NewEnumerator3
    Mesh Uniform Scale=3
    Sprite Rotation Mode=::NewEnumerator0
    Sprite Size=&SelectVector2DFromArray(
     Direct Array Index=$Particles.PartitionParticles.Partition
     Vector 2D Selection Array=DataInterface() {propertyValues {FloatData [<115 75>, <115 75>, <75 50>, <75 50>, <75 50>, <75 50>]}}
    )
    Sprite Size Mode=::NewEnumerator4
    UsePositionOffset=false
    Write Material Random=true
   )
   /SetVariables(Particles.Streak_01_Size:Vector2f=<300 15>) {isSetParameters=true, module=SetVariables_#1}
   &SystemLocation(
    MaskSpawnGroup=false
    Offset=&SelectVectorFromArray(
     Direct Array Index=$Particles.PartitionParticles.Partition
     Vector Selection Array=DataInterface() {propertyValues {FloatData [<0 -370 -29.5>, <0 -496 -22.5>, <0 379 -38>, <0 343 -25>, <2 328 -24.5>, <0 300.5 -32>]}}
    )
   )
   &SpriteFacingAndAlignment(WriteAlignment=false, WriteFacing=true)
  ]
  particleUpdateScript [
   &ParticleState(
    Kill Particles When Lifetime Has Elapsed=false
    Let Infinitely Lived Particles Die When Emitter Deactivates=true
    Loop Particles Lifetime=true
   )
   &self:ScratchModule_01(
    NumberOfRings=5
    OcclusionSize=10
    Position:NiagaraPosition=&TransformPosition(Position=$Particles.Position, Source Space=::Local, Destination Space=::World)
    SamplesPerRing=5
    SystemX Sample Offset=5
   ) {module=ScratchModule_01}
   /SetVariables(Particles.CameraScale:NiagaraFloat=&MakeFloatFromVector(
     Vector=&Multiply_VectorByFloat(
      Vector=&ReturnCameraVector(Camera Vectors=::NewEnumerator4, Coordinate Space=::Simulation, Origin Position=$Particles.Position)
      Float=&DotProduct(
       Vector A=<1 0 -1.5>
       Vector B=&ReturnCameraVector(Camera Vectors=::NewEnumerator4, Coordinate Space=::Simulation, Origin Position=$Particles.Position)
      )
     )
     Channel=::NewEnumerator0
    )) {isSetParameters=true}
   &ScaleSpriteSize(
    Uniform Curve Index=$Particles.CameraScale
    Uniform Curve Sprite Scale=DataInterface() {propertyValues {Curve {keys [
        {arriveTangent=4.7849, interpMode=RCIM_Cubic, leaveTangent=4.7849, tangentMode=RCTM_User, time=0, value=0.5}
        {arriveTangent=0.9156, interpMode=RCIM_Cubic, leaveTangent=0.9156, tangentMode=RCTM_User, time=0.3, value=1.275}
        {tangentMode=RCTM_User, time=1, value=1.5}
       ]}}}
   )
   &ScaleColor(Scale Alpha=$Particles.Visibility, Scale RGB=&VectorFromFloat(Value=3), ScaleA=true, ScaleRGB=true) {module=ScaleColor002}
   &ScaleColor(
    Scale Alpha=&FloatFromCurve(
     FloatCurve=DataInterface() {propertyValues {Curve {keys [
         {interpMode=RCIM_Cubic, tangentMode=RCTM_User, time=0, value=0}
         {arriveTangent=0.4636, interpMode=RCIM_Cubic, leaveTangent=0.4636, tangentMode=RCTM_User, time=0.2, value=0.8}
         {arriveTangentWeight=1.2316, interpMode=RCIM_Cubic, tangentMode=RCTM_User, time=1, value=1}
        ]}}}
     CurveIndex=$Particles.CameraScale
     Scale Curve=&Lerp_Float(
      A=0.7
      B=1
      Alpha=&ClampFloat(Float=&Waveform(
        Global Phase=&Multiply_Float(A=$Emitter.Age, B=0.2)
        "Waveform [1]"=::NewEnumerator0
        "[1] Frequency"=2
        "[1] Amplitude Min/Max"=<0 1>
       ))
     )
    )
    Scale RGB=&Multiply_VectorByFloat(Vector=<1 1 1>, Float=$Particles.CameraScale)
    ScaleA=true
    ScaleRGB=true
   )
   /SetVariables(
    Particles.GlowSpriteColor:LinearColor=&MultiplyLinearColor(
     Linear Color=&SelectLinearColorFromArray(
      Array Sampling Mode=::NewEnumerator1
      Direct Array Index=$Particles.PartitionParticles.Partition
      Color Selection Array=DataInterface() {propertyValues {ColorData [<rgba 1 0.5764 0.2812 0.3>, <rgba 1 0.5764 0.2812 0.3>, <rgba 1 0.5764 0.2812 0.2>]}}
      Random Seed=?
     )
     Scale Factor=&MakeFloat4FromVectorAndFloat(
      "Vector (RGB)"=<1 1 1>
      "Float (Alpha)"=&Lerp_Float(
       A=0.9
       B=1
       Alpha=&ClampFloat(Float=&Waveform(
         Global Phase=&Multiply_Float(A=$Emitter.Age, B=0.2)
         "Waveform [1]"=::NewEnumerator0
         "[1] Frequency"=2
         "[1] Amplitude Min/Max"=<0 1>
        ))
      )
     )
    )
    Particles.GlowSpriteSize:Vector2f=&SelectVector2DFromArray(
     Direct Array Index=$Particles.PartitionParticles.Partition
     Vector 2D Selection Array=DataInterface() {propertyValues {FloatData [<250 250>, <250 250>, <150 150>]}}
    )
   ) {isSetParameters=true, module=SetVariables_#2}
  ]
  rendererClasses [&NiagaraSpriteRendererProperties]
  renderers [
   {rendererClass=&NiagaraSpriteRendererProperties}
   {rendererClass=&NiagaraSpriteRendererProperties, rendererIndex=1}
   {rendererClass=&NiagaraSpriteRendererProperties, rendererIndex=2}
  ]
 }
 Car_LensFlare_LOW {
  emitterUpdateScript [
   &EmitterState()
   &SpawnBurst_Instantaneous(Spawn Count=6, Use Spawn Probability=false, UseLoopCountLimit=true)
  ]
  particleSpawnScript [
   &PartitionParticles(Number Of Partitions=6)
   &InitializeParticle(
    Color=<rgba 2 1.1529 0.5625 0.8>
    Mesh Scale Mode=::NewEnumerator3
    Mesh Uniform Scale=3
    Sprite Rotation Mode=::NewEnumerator0
    Sprite Size=&SelectVector2DFromArray(
     Direct Array Index=$Particles.PartitionParticles.Partition
     Vector 2D Selection Array=DataInterface() {propertyValues {FloatData [<115 75>, <115 75>, <75 50>, <75 50>, <75 50>, <75 50>]}}
    )
    Sprite Size Mode=::NewEnumerator4
    UsePositionOffset=false
    Write Material Random=true
   )
   /SetVariables(Particles.Streak_01_Size:Vector2f=<300 8>) {isSetParameters=true, module=SetVariables_#1}
   &SystemLocation(
    MaskSpawnGroup=false
    Offset=&SelectVectorFromArray(
     Direct Array Index=$Particles.PartitionParticles.Partition
     Vector Selection Array=DataInterface() {propertyValues {FloatData [<0 -370 -29.5>, <0 -496 -22.5>, <0 379 -38>, <0 343 -25>, <2 328 -24.5>, <0 300.5 -32>]}}
    )
   )
   &SpriteFacingAndAlignment(WriteAlignment=false, WriteFacing=true)
  ]
  particleUpdateScript [
   &ParticleState(
    Kill Particles When Lifetime Has Elapsed=false
    Let Infinitely Lived Particles Die When Emitter Deactivates=true
    Loop Particles Lifetime=true
   )
   /SetVariables(
    Particles.GlowSpriteColor:LinearColor=&MultiplyLinearColor(
     Linear Color=&SelectLinearColorFromArray(
      Array Sampling Mode=::NewEnumerator1
      Direct Array Index=$Particles.PartitionParticles.Partition
      Color Selection Array=DataInterface() {propertyValues {ColorData [<rgba 1 0.5764 0.2812 0.3>, <rgba 1 0.5764 0.2812 0.3>, <rgba 1 0.5764 0.2812 0.2>]}}
      Random Seed=?
     )
     Scale Factor=&MakeFloat4FromVectorAndFloat(
      "Vector (RGB)"=<1 1 1>
      "Float (Alpha)"=&Lerp_Float(
       A=0.85
       B=1
       Alpha=&ClampFloat(Float=&Waveform(
         Global Phase=&Multiply_Float(A=$Emitter.Age, B=0.1)
         "Waveform [1]"=::NewEnumerator0
         "[1] Frequency"=2
         "[1] Amplitude Min/Max"=<0 1>
        ))
      )
     )
    )
    Particles.GlowSpriteSize:Vector2f=&SelectVector2DFromArray(
     Direct Array Index=$Particles.PartitionParticles.Partition
     Vector 2D Selection Array=DataInterface() {propertyValues {FloatData [<250 250>, <250 250>, <150 150>]}}
    )
   ) {isSetParameters=true, module=SetVariables_#2}
   &ScaleColor(
    Scale Alpha=&Lerp_Float(
     A=0.65
     B=1
     Alpha=&ClampFloat(Float=&Waveform(
       Global Phase=&Multiply_Float(A=$Emitter.Age, B=0.2)
       "Waveform [1]"=::NewEnumerator0
       "[1] Frequency"=2
       "[1] Amplitude Min/Max"=<0 1>
      ))
    )
    ScaleA=true
    ScaleRGB=true
   )
  ]
  rendererClasses [&NiagaraSpriteRendererProperties]
  renderers [
   {rendererClass=&NiagaraSpriteRendererProperties}
   {rendererClass=&NiagaraSpriteRendererProperties, rendererIndex=1}
   {rendererClass=&NiagaraSpriteRendererProperties, rendererIndex=2}
  ]
 }
 Foreground_Fog {
  emitterUpdateScript [
   &EmitterState()
   &SpawnRate(SpawnRate=4, Use Spawn Probability=false)
  ]
  particleSpawnScript [
   &InitializeParticle(
    Color Maximum=<rgba 0.5 0.4449 0.3854 0.1>
    Color Minimum=<rgba 0.5 0.4586 0.4141 0.01>
    Color Mode=::NewEnumerator2
    Lifetime Max=10
    Lifetime Min=7
    Lifetime Mode=::NewEnumerator1
    Sprite Rotation Mode=::NewEnumerator2
    Sprite Size Mode=::NewEnumerator1
    Sprite UV Mode=::NewEnumerator3
    Uniform Sprite Size Max=300
    Uniform Sprite Size Min=200
    UsePositionOffset=false
    Write Material Random=true
   )
   &ShapeLocation(
    Box Midpoint=&RandomRangeVector(Minimum=<0 0 0>, Maximum=<0.5 1 1>)
    Box Size=<400 1000 50>
    MaskSpawnGroup=false
    Offset=<0 250 -50>
    Offset Coordinate Space=::World
    Offset Mode=::NewEnumerator0
    Override Seed=false
    Shape Primitive=::NewEnumerator2
    Surface Only=false
    UseOwnerScale=true
   )
   &DynamicMaterialParameters(Write Parameter Index 0=false)
   &AddVelocity(
    Velocity=&RandomRangeVector(Minimum=<0 -1 0.5>, Maximum=<0 -3 -2>)
    Velocity Speed Scale=&RandomRangeFloat(Minimum=5, Maximum=30)
   )
   &CameraOffset(Camera Offset Amount=150)
  ]
  particleUpdateScript [
   &ParticleState()
   &SpriteRotationRate(Rotation Rate=&RandomRangeFloat(Minimum=1.5, Maximum=2.5))
   &ScaleSpriteSize(Uniform Curve Sprite Scale=DataInterface() {propertyValues {Curve {keys [{tangentMode=RCTM_User, time=0, value=0.75}, {tangentMode=RCTM_User, time=1, value=1}]}}})
   &ScaleColor(
    Scale Alpha=&FloatFromCurve(FloatCurve=DataInterface() {propertyValues {Curve {keys [<at 0 0 RCIM_Cubic>,<at 0.5 1 RCIM_Cubic>,<at 1 0 RCIM_Cubic>]}}}, Scale Curve=0.9)
    ScaleA=true
    ScaleRGB=true
   )
   &SolveForcesAndVelocity(Clamp Velocity=false, Limit Acceleration=false)
   &DynamicMaterialParameters(
    Index 0 Param 1=&FloatFromCurve(FloatCurve=DataInterface() {propertyValues {Curve {keys [
         {interpMode=RCIM_Cubic, tangentMode=RCTM_User, time=0, value=1}
         {interpMode=RCIM_Cubic, tangentMode=RCTM_User, time=0.1, value=0}
         {interpMode=RCIM_Cubic, tangentMode=RCTM_User, time=1, value=1}
        ]}}})
    Index 0 Param 2=&FloatFromCurve(FloatCurve=DataInterface() {propertyValues {Curve {keys [
         {interpMode=RCIM_Cubic, tangentMode=RCTM_User, time=0, value=1}
         {interpMode=RCIM_Cubic, tangentMode=RCTM_User, time=0.25, value=0.5}
         {interpMode=RCIM_Cubic, tangentMode=RCTM_User, time=1, value=1}
        ]}}})
    Index 0 Param 3=400
    Module.Param0WriteEnabled=true
    Module.Param1WriteEnabled=true
    Module.Param2WriteEnabled=true
    Module.Param3WriteEnabled=true
   ) {module=DynamicMaterialParameters001}
   &SubUVAnimation(
    Frames Per Second:NiagaraInt32=&RandomRangeInt(Minimum=-10, Maximum=-12)
    Playback Mode:ENiagara_FPSPlayrate=::NewEnumerator1
    Playback Scale Curve=DataInterface() {propertyValues {Curve {keys [
        {interpMode=RCIM_Cubic, tangentMode=RCTM_User, time=0.25, value=1}
        {arriveTangent=-1.2656, interpMode=RCIM_Cubic, leaveTangent=-1.2656, tangentMode=RCTM_User, time=1, value=0}
       ]}}}
    Random Start Frame=true
    "Sprite Renderer "=DataInterface() {propertyValues {SpriteRenderer=&self:SimpleSpriteBurst.NiagaraSpriteRendererProperties_0}}
    Start Frame Offset:NiagaraInt32=&RandomRangeInt(Minimum=0, Maximum=16)
    Start Frame Range Override=&RandomRangeInt(Minimum=0, Maximum=0)
    SubUV Animation Mode:ENiagara_SubUVLookupModeV2=::NewEnumerator3
    UseEndFrame=true
    UsePlayRateScale=false
    UseStartFrame=true
   )
   &SpriteFacingAndAlignment(Sprite Facing=<0 1 0>, WriteAlignment=false, WriteFacing=true)
  ]
  rendererClasses [&NiagaraSpriteRendererProperties]
  renderers [{rendererClass=&NiagaraSpriteRendererProperties}]
 }
 LightShafts {
  emitterUpdateScript [
   &EmitterState()
   &SpawnBurst_Instantaneous(Spawn Count=2, Use Spawn Probability=false, UseLoopCountLimit=true)
  ]
  particleSpawnScript [
   &PartitionParticles()
   &InitializeParticle(
    Color=<rgba 1.7 0.9799 0.4781 0.4>
    Sprite Rotation Angle=&SelectFloatFromArray(
     Direct Array Index=$Particles.PartitionParticles.Partition
     Float Selection Array=DataInterface() {propertyValues {FloatData [0.72,0.72,0.27,0.27,0.27,0.27]}}
    )
    Sprite Rotation Mode=::NewEnumerator1
    Sprite Size=&SelectVector2DFromArray(
     Direct Array Index=$Particles.PartitionParticles.Partition
     Vector 2D Selection Array=DataInterface() {propertyValues {FloatData [<200 400>, <200 400>, <50 150>, <50 150>, <50 150>, <50 150>]}}
    )
    Sprite Size Mode=::NewEnumerator4
    UsePositionOffset=false
    Write Material Random=true
   )
   &CameraOffset(Camera Offset Amount=&RandomRangeFloat(Minimum=0, Maximum=15, Evaluation Type=::NewEnumerator1))
   &SystemLocation(
    MaskSpawnGroup=false
    Offset=&SelectVectorFromArray(
     Direct Array Index=$Particles.PartitionParticles.Partition
     Vector Selection Array=DataInterface() {propertyValues {FloatData [<0 -370 -29.5>, <0 -496 -22.5>]}}
    )
   )
   /SetVariables(Particles.SubImageIndex=1) {isSetParameters=true}
   &SpriteFacingAndAlignment(Sprite Facing=<1 -1 0>, WriteAlignment=false, WriteFacing=true)
  ]
  particleUpdateScript [
   &ParticleState(
    Kill Particles When Lifetime Has Elapsed=false
    Let Infinitely Lived Particles Die When Emitter Deactivates=true
    Loop Particles Lifetime=true
   )
   &ScaleColor(
    Scale Alpha=&Lerp_Float(
     A=0.6
     B=1
     Alpha=&ClampFloat(Float=&Waveform(
       Global Phase=&Multiply_Float(A=$Emitter.Age, B=0.15)
       "Waveform [1]"=::NewEnumerator0
       "[1] Frequency"=2
       "[1] Amplitude Min/Max"=<0 1>
       "Waveform [2] Blend"=::NewEnumerator1
      ))
    )
    ScaleA=true
    ScaleRGB=true
   )
  ]
  rendererClasses [&NiagaraSpriteRendererProperties]
  renderers [{rendererClass=&NiagaraSpriteRendererProperties}]
 }
 Smoke_Mouth {
  emitterUpdateScript [
   &EmitterState()
   &SpawnRate(SpawnRate=4, Use Spawn Probability=false)
  ]
  particleSpawnScript [
   &InitializeParticle(
    Color Maximum=<rgba 0.5003 0 0.6042 1>
    Color Minimum=<rgba 1 0 0.6842 1>
    Color Mode=::NewEnumerator2
    Lifetime Max=8
    Lifetime Min=4
    Lifetime Mode=::NewEnumerator1
    Mesh Scale Mode=::NewEnumerator3
    Mesh Uniform Scale=3
    Sprite Rotation Mode=::NewEnumerator2
    Sprite Size Mode=::NewEnumerator1
    Uniform Sprite Size Max=225
    Uniform Sprite Size Min=150
    UsePositionOffset=false
    Write Material Random=true
   ) {module=InitializeParticle001}
   &PartitionParticles()
   /SetVariables(Particles.RandomFrame=&RandomRangeInt(Minimum=0, Maximum=63)) {isSetParameters=true}
   &DynamicMaterialParameters(
    Index 0 Param 1=&RandomRangeFloat(Minimum=0.2, Maximum=0.5)
    Module.Param0WriteEnabled=true
    Module.Param1WriteEnabled=true
    Module.Param2WriteEnabled=true
    Module.Param3WriteEnabled=true
   )
   &CameraOffset(Camera Offset Amount=&RandomRangeFloat(Minimum=0, Maximum=2))
   &SystemLocation(MaskSpawnGroup=false, Offset=<-100 -165 138>)
   &AddVelocity(
    Cone Angle=15
    Cone Axis=&SelectVectorFromArray(
     Direct Array Index=$Particles.PartitionParticles.Partition
     Vector Selection Array=DataInterface() {propertyValues {FloatData [<1 1 0>, <1 -1 0>]}}
    )
    Inner Cone Angle=10
    Module.UseConeFalloff=false
    Module.UseVelDistribution=false
    Override Seed=false
    Velocity Mode=::NewEnumerator2
    Velocity Speed=&RandomRangeFloat(Minimum=85, Maximum=110)
   )
  ]
  particleUpdateScript [
   &ParticleState()
   &FN_SubUVAnimation(
    End Frame=&Add_Integer(A=0, B=63)
    SubUV Animation Mode:ENiagara_SubUVLookupMode=::NewEnumerator1
    SubUV Curve=DataInterface() {propertyValues {Curve {keys [
        {arriveTangent=0.7286, interpMode=RCIM_Cubic, leaveTangent=0.7286, tangentMode=RCTM_User, time=0, value=0}
        {arriveTangent=0.5299, tangentMode=RCTM_User, time=1, value=1}
       ]}}}
    SubUV Loop Count:NiagaraFloat=&RandomRangeFloat(Minimum=1, Maximum=1)
   )
   &ScaleSpriteSize(
    Uniform Curve Scale=0.5
    Uniform Curve Sprite Scale=DataInterface() {propertyValues {Curve {keys [
        {arriveTangent=2.6136, interpMode=RCIM_Cubic, leaveTangent=2.6136, tangentMode=RCTM_User, time=0, value=0.25}
        {arriveTangent=0.0897, interpMode=RCIM_Cubic, leaveTangent=0.0897, tangentMode=RCTM_User, time=1, value=1}
       ]}}}
   )
   &ScaleColor(
    Scale Alpha=&FloatFromCurve(
     FloatCurve=DataInterface() {propertyValues {Curve {keys [
         {interpMode=RCIM_Cubic, tangentMode=RCTM_User, time=0, value=0}
         <at 0.05 1.05>
         {arriveTangent=-1.1053, interpMode=RCIM_Cubic, leaveTangent=-1.1053, time=0.7, value=1.05}
         {tangentMode=RCTM_User, time=1, value=0}
        ]}}}
     Scale Curve=0.015
    )
    Scale RGB=&Multiply_VectorByFloat(
     Vector=<1 1 1>
     Float=&FloatFromCurve(
      FloatCurve=DataInterface() {propertyValues {Curve {keys [
          {interpMode=RCIM_Cubic, tangentMode=RCTM_User, time=0, value=0.5}
          {arriveTangent=0.0148, interpMode=RCIM_Cubic, leaveTangent=0.0148, tangentMode=RCTM_User, time=0.2, value=1.35}
          {arriveTangent=-1.6875, interpMode=RCIM_Cubic, leaveTangent=-1.6875, time=0.8, value=1}
          <at 1 0>
         ]}}}
      Scale Curve=3
     )
    )
    ScaleA=true
    ScaleRGB=true
   ) {module=ScaleColor001}
   &DynamicMaterialParameters(
    Index 0 Param 1=&FloatFromCurve(
     FloatCurve=DataInterface() {propertyValues {Curve {keys [
         {tangentMode=RCTM_User, time=0, value=0}
         {interpMode=RCIM_Cubic, tangentMode=RCTM_User, time=0.8, value=0.125}
         {tangentMode=RCTM_User, time=1, value=1}
        ]}}}
     Scale Curve=0
    )
    Index 0 Param 3=1
    Module.Index0Param0Write=true
    Module.Index0Param1Write=true
    Module.Index0Param2Write=true
    Module.Index0Param3Write=true
   ) {module=DynamicMaterialParameters001}
   &GravityForce(Gravity=<0 0 -35>)
   &Drag(Drag=1.5, Use Linear Drag=true, Use Rotational Drag=true)
   &SolveForcesAndVelocity(Clamp Velocity=true, Limit Acceleration=false, Speed Limit=250)
   &SpriteFacingAndAlignment(Sprite Facing=<0 1 0>, WriteAlignment=false, WriteFacing=true)
  ]
  rendererClasses [&NiagaraSpriteRendererProperties]
  renderers [{rendererClass=&NiagaraSpriteRendererProperties}]
 }
}
summary {
 systemName=NS_S30_Frontend_Ambient
 emitters [
  {emitterName=Foreground_Fog, rendererClasses [&NiagaraSpriteRendererProperties]}
  {emitterName=Smoke_Mouth, rendererClasses [&NiagaraSpriteRendererProperties]}
  {emitterName=Car_LensFlare, simTarget=GPUComputeSim, rendererClasses [&NiagaraSpriteRendererProperties]}
  {emitterName=Car_LensFlare_LOW, rendererClasses [&NiagaraSpriteRendererProperties]}
  {emitterName=LightShafts, rendererClasses [&NiagaraSpriteRendererProperties]}
 ]
 userVariables []
}
